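; BP_CreatureManager is a Creative device that overrides stats and behavior for a chosen creature type.
; Setup copies user options into the native BuildingProp_CreatureManager selections, converting aggro range to world units and inverting the knockback option into immunity.
; Option changes refresh a preview skeletal mesh once it loads.
; Events and minigame start or end (on authority) toggle the device, and managed creature deaths broadcast an elimination event with the instigator.
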(event EventOnWorldReady
  (Visibility|SetVisibilityComponentsWithCollision (Variables|Default|GetVisibleInGameComponent) (Utilities|Map|MakeMap (Variables|Debug|GetStaticMeshComponent) "QueryAndPhysics" (Variables|Default|GetAITypeMesh) "NoCollision") true))

(event EventSetupConfigurables
  (CallFunction|SetPresetValues))

(event OnAnyPropertyChanged(ToyOptionsComponent)
  (Variables|BuildingProp_CreatureManager|SetSelectedAIType (Utilities|Array|Get(acopy) (Variables|CustomizeOptions|GetAITypeOptions) (Variables|Default|GetAITypePresetIndex)))
  (CallFunction|LoadSkeletalMeshDisplayOption))

(event OnPropertiesLoaded(ToyOptionsComponent)
  (Variables|BuildingProp_CreatureManager|SetSelectedAIType (Utilities|Array|Get(acopy) (Variables|CustomizeOptions|GetAITypeOptions) (Variables|Default|GetAITypePresetIndex)))
  (CallFunction|LoadSkeletalMeshDisplayOption))

(event EventSkeletalMeshLoadComplete
  (params (Material :type Material%20Interface%20Object%20Reference) (Mesh :type Skeletal%20Mesh%20Object%20Reference))
  (CallFunction|UpdateVisibleSkeletalMesh Material Mesh))

(event OnVisibilityChanged(VisibleInGameComponent)
  (params (Visible :type Boolean))
  (CallFunction|AdjustCollisionOfStaticMesh (not Visible) (Variables|Debug|GetStaticMeshComponent)))

(event Custom|EnableEvent
  (params (Instigator :type Controller%20Object%20Reference))
  (CallFunction|NotifyDeviceEnableChanged true))

(event MinigameStartedDelegate(FortMinigameProgress)
  (Networking|SwitchHasAuthority
    (:Authority
      (CallFunction|NotifyDeviceEnableChanged (Variables|Default|GetEnabledOnMinigameStart)))))

(event Custom|DisableEvent
  (params (Instigator :type Controller%20Object%20Reference))
  (CallFunction|NotifyDeviceEnableChanged false))

(event MinigameEndedDelegate(FortMinigameProgress)
  (Networking|SwitchHasAuthority
    (:Authority
      (CallFunction|NotifyDeviceEnableChanged false))))

(event EventOnManagedCreatureDied
  (params (InstigatingController :type Controller%20Object%20Reference) (DamagedActor :type Actor%20Object%20Reference))
  (GameplayEvent|BroadcastEvent (Variables|UserOptions|Events|GetOn%20a%20Matching%20Creature%20Type%20is%20Eliminated) 0 (Utilities|Struct|MakeGameplayEventManagedCreatureEliminated InstigatingController)))

(fn SetPresetValues
  (Variables|BuildingProp_CreatureManager|SetSelectedAIType (Utilities|Array|Get(acopy) (Variables|CustomizeOptions|GetAITypeOptions) (Variables|Default|GetAITypePresetIndex)))
  (Variables|BuildingProp_CreatureManager|SetSelectedOverrideHealth (Variables|Default|GetAIHealthPresetValue))
  (Variables|BuildingProp_CreatureManager|SetSelectedOverrideScoreValue (Variables|Default|GetAIScorePresetValue))
  (Variables|BuildingProp_CreatureManager|SetSelectedOverrideAggroDistance (* (Variables|Default|GetAIAggroRangePresetValue) 512.0))
  (Variables|BuildingProp_CreatureManager|SetSelectedOverrideDamage (Variables|Default|GetAIDamagePresetValue))
  (Variables|BuildingProp_CreatureManager|SetSelectedEnvironmentalDamageOverride (Variables|Default|GetAIEnvironmentalDamagePresetValue))
  (Variables|BuildingProp_CreatureManager|SetSelectedMovementSpeedMultiplier (Variables|Default|GetAIMovementSpeedMultiplierPresetValue))
  (Variables|BuildingProp_CreatureManager|SetSelectedOverrideScoreDistribution (Variables|Default|GetAIScoreDistributionPresetValue))
  (Variables|BuildingProp_CreatureManager|SetbImmuneToWeaponKnockback (not (Variables|Default|GetAIAllowWeaponKnockbackPresetValue)))
  (Variables|BuildingProp_CreatureManager|SetSelectedAffectedTargetsType (Variables|Default|GetAffectedCreaturesValue)))

(fn UpdateVisibleSkeletalMesh
  (params (Material :type Material%20Interface%20Object%20Reference) (Mesh :type Skeletal%20Mesh%20Object%20Reference))
  (Components|SkinnedMesh|SetSkinnedAssetandUpdate (Variables|Default|GetAITypeMesh) (Variables|GetMesh) true)
  (Rendering|Material|SetMaterial (Variables|Default|GetAITypeMesh) 0 (Variables|GetMaterial)))

(fn ConstructionScript)
